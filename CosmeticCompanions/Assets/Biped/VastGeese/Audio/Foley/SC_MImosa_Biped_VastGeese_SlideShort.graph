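# Sound cue for the short-slide foley of the VastGeese biped cosmetic companion.
# A quality-level node routes the first two levels to a random pick of eight wave variations and the last two to a smaller four-wave subset, all equally weighted.
# It plays in the creature pets sound class with companion foley attenuation and shares the biped footstep concurrency group.
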SoundCue /CosmeticCompanions/Assets/Biped/VastGeese/Audio/Foley/SC_MImosa_Biped_VastGeese_SlideShort
root=QualityLevel_0
nodes [
 {id=QualityLevel_0, children [Random_0,Random_0,Random_1,Random_1]}
 {
  id=Random_0
  children [WavePlayer_10, WavePlayer_11, WavePlayer_12, WavePlayer_13, WavePlayer_14, WavePlayer_15, WavePlayer_0, WavePlayer_1]
  properties {weights [1,1,1,1,1,1,1,1]}
 }
 {id=Random_1, children [WavePlayer_2, WavePlayer_17, WavePlayer_18, WavePlayer_19], properties {weights [1,1,1,1]}}
 {
  id=WavePlayer_0
  properties {soundWaveAssetPtr=/CosmeticCompanions/Assets/Biped/VastGeese/Audio/Foley/Wavs/Mimosa_Biped_VastGeese_SlideShort_08}
 }
 {
  id=WavePlayer_1
  properties {soundWaveAssetPtr=/CosmeticCompanions/Assets/Biped/VastGeese/Audio/Foley/Wavs/Mimosa_Biped_VastGeese_SlideShort_07}
 }
 {
  id=WavePlayer_10
  properties {soundWaveAssetPtr=/CosmeticCompanions/Assets/Biped/VastGeese/Audio/Foley/Wavs/Mimosa_Biped_VastGeese_SlideShort_06}
 }
 {
  id=WavePlayer_11
  properties {soundWaveAssetPtr=/CosmeticCompanions/Assets/Biped/VastGeese/Audio/Foley/Wavs/Mimosa_Biped_VastGeese_SlideShort_05}
 }
 {
  id=WavePlayer_12
  properties {soundWaveAssetPtr=/CosmeticCompanions/Assets/Biped/VastGeese/Audio/Foley/Wavs/Mimosa_Biped_VastGeese_SlideShort_04}
 }
 {
  id=WavePlayer_13
  properties {soundWaveAssetPtr=/CosmeticCompanions/Assets/Biped/VastGeese/Audio/Foley/Wavs/Mimosa_Biped_VastGeese_SlideShort_03}
 }
 {
  id=WavePlayer_14
  properties {soundWaveAssetPtr=/CosmeticCompanions/Assets/Biped/VastGeese/Audio/Foley/Wavs/Mimosa_Biped_VastGeese_SlideShort_02}
 }
 {
  id=WavePlayer_15
  properties {soundWaveAssetPtr=/CosmeticCompanions/Assets/Biped/VastGeese/Audio/Foley/Wavs/Mimosa_Biped_VastGeese_SlideShort_01}
 }
 {
  id=WavePlayer_17
  properties {soundWaveAssetPtr=&Mimosa_Biped_VastGeese_SlideShort_03}
 }
 {
  id=WavePlayer_18
  properties {soundWaveAssetPtr=&Mimosa_Biped_VastGeese_SlideShort_02}
 }
 {
  id=WavePlayer_19
  properties {soundWaveAssetPtr=&Mimosa_Biped_VastGeese_SlideShort_01}
 }
 {
  id=WavePlayer_2
  properties {soundWaveAssetPtr=&Mimosa_Biped_VastGeese_SlideShort_07}
 }
]
properties {
 attenuationSettings=/Game/Sounds/Attenuation/AS_Companion_Foley
 maxDistance=1200
 priority=0
 soundClassObject=&Fort_Creatures_Pets
 volumeMultiplier=0.9
 concurrencySet [/CosmeticCompanions/Assets/Audio_Shared/Foley/Biped/CONC_Mimosa_Biped_Footstep]
}
